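# Sound cue for the Disguise gadget's initiate/active gameplay cue, layering a disappear sound with a reappear sound.
# A mixer plays the disappear wave immediately and the reappear wave after a short fixed delay.
# It routes through the gadget items sound class with the disguise equip attenuation, using the disguise concurrency set alongside a one-per-owner concurrency override.
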SoundCue /FNGameplayCues/Athena/Sounds/Gadgets/Disguise/Disguise_Initiate_Active_Cue
root=Mixer_0
nodes [
 {id=Delay_0, children [WavePlayer_1], properties {delayMax=0.3, delayMin=0.3}}
 {id=Mixer_0, children [WavePlayer_2,Delay_0], properties {inputVolume [1,1]}}
 {id=WavePlayer_1, properties {soundWaveAssetPtr=/Game/Athena/Sounds/Gadgets/Disguise/Item_Disguise_Reappear_01}}
 {id=WavePlayer_2, properties {soundWaveAssetPtr=/Game/Athena/Sounds/Gadgets/Disguise/Item_Disguise_Disappear_01}}
]
properties {
 attenuationSettings=/Game/Athena/Sounds/Gadgets/Disguise/Disguise_Equip_Att
 maxDistance=1712
 soundClassObject=&Fort_Gadgets_Items
 volumeMultiplier=0.65
 concurrencyOverrides {bLimitToOwner=true, maxCount=1}
 concurrencySet [/Game/Athena/Sounds/Gadgets/Disguise/Disguise_Concurrency]
}
