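# Sound cue for the disguise gadget's initiate-scan action in Athena, played as a gameplay cue sound.
# It plays a single scan wave directly through one wave player, with no randomization or modulation.
# It borrows the disguise equip attenuation asset and routes to the gadget items sound class.
# An overridden concurrency limit caps simultaneous instances and stops the oldest.
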SoundCue /FNGameplayCues/Athena/Sounds/Gadgets/Disguise/Disguise_Initiate_Scan_Cue
root=WavePlayer_0
nodes [{id=WavePlayer_0, properties {soundWaveAssetPtr=&/Item_Disguise_Initiate_Scan_01}}]
properties {
 attenuationSettings=/Game/Athena/Sounds/Gadgets/Disguise/Disguise_Equip_Att
 bOverrideConcurrency=true
 maxDistance=1712
 soundClassObject=&Fort_Gadgets_Items
 volumeMultiplier=1.15
 concurrencyOverrides {maxCount=4, resolutionRule=StopOldest}
}
